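# Translucent unlit Niagara sprite material for an energy effect in the CreatureRise ability VFX.
# Noise-distorted V coordinates offset by a TimePanner dynamic parameter sample a curve-atlas gradient, scaled by PulseEmissive to brighten magenta-tinted panning noise and multiplied by particle color.
# Opacity combines a curve-remapped sphere mask of noise-varied radius, panning smoke detail set by WithoutNoiseDetail, a vertical UV fade, depth fade and particle alpha.
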Material /NPCLibrary/Abilities/CreatureRise/VFX/M_Chrome_Energy_T
graph {
 properties {bUsedWithNiagaraSprites=true, blendMode=BLEND_Translucent, pixelDepthOffsetMode=PDOM_Legacy, shadingModel=MSM_Unlit, usedShadingModels=Unlit}
 nodes [
  Constant() {r=0.5}
  Constant() {r=1}
  n1 = TextureCoordinate()
  n2 = TextureSampleParameter2D() {parameterName=DistortionTexture, samplerType=SAMPLERTYPE_Grayscale, texture=&LowResBlurredNoise}
  n3 = Add(
   A=Add(
    A=MaterialFunctionCall(Float2=@n1) {materialFunction=&BreakOutFloat2Components}.G
    B=DynamicParameter() {paramNames [WithoutNoiseDetail, TimePanner, PulseEmissive, Param4]}.TimePanner
   )
   B=Multiply(A=@n2.R, B=ScalarParameter() {defaultValue=0.1, parameterName=RingUV_Distortion})
  )
  n4 = CurveAtlasRowParameter(CurveTime=Frac(None=Multiply(A=@n3))) {
   atlas=/Game/Athena/Items/AwfulDuck/FX/CA_AwfulDuck
   curve=/Game/Athena/Items/AwfulDuck/FX/Curve_AwfulDuck_03
   defaultValue=2
   parameterName=RingGradient
  }
  n5 = Multiply(A=@n4.R, B=DynamicParameter() {paramNames [WithoutNoiseDetail, TimePanner, PulseEmissive, Param4]}.PulseEmissive)
  MaterialFunctionCall(LDR Input=@n4.R, HDR Tint=@n5) {materialFunction=&DeriveHDRfromLDR}
  n6 = Panner(Coordinate=Multiply(A=TextureCoordinate(), B=Constant2Vector() {g=0.02, r=1})) {speedY=-0.5}
  n7 = TextureSample(UVs=@n6) {texture=&T_NoisePack_01}
  n8 = ScalarParameter() {defaultValue=0.5}
  n9 = Multiply(
   A=Multiply(A=Desaturation(None=@n7.RGB, Fraction=@n8), B=Constant3Vector() {constant=<rgba 1 0.3 1 0>})
   B=ScalarParameter() {defaultValue=3, parameterName=Param_1}
  )
  n10 = Multiply(A=@n5, B=@n9)
  n11 = ParticleColor()
  MaterialFunctionCall(Emissive=Multiply(A=Add(A=@n10, B=@n9), B=@n11.RGB)) {materialFunction=&MF_FX_EmissiveAdjustments}
  n12 = TextureCoordinate()
  n13 = Constant2Vector() {g=0.5, r=1}
  n14 = CurveAtlasRowParameter(CurveTime=SphereMask(A=Multiply(A=@n12, B=@n13), B=Constant() {r=0.5}, Radius=LinearInterpolate(Alpha=@n7.R) {constA=0.4, constB=0.55}, Hardness=Constant())) {
   atlas=&CA_AwfulDuck
   curve=/Game/Athena/Items/AwfulDuck/FX/Curve_ExpCurve
   defaultValue=10
   parameterName=SphereGradient
  }
  n15 = Multiply(
   A=LinearInterpolate(
    A=DynamicParameter() {paramNames [WithoutNoiseDetail, TimePanner, PulseEmissive, Param4]}.WithoutNoiseDetail
    Alpha=TextureSample(UVs=@n6) {texture=&T_Inky_Smoke_Tile}.R
   )
   B=Multiply(A=@n14.R)
  )
  n16 = MaterialFunctionCall(Float2=TextureCoordinate()) {materialFunction=&BreakOutFloat2Components}
  n17 = Multiply(A=@n16.G)
  n18 = Multiply(A=OneMinus(None=@n16.G), B=Saturate(None=Multiply(A=@n17) {constB=2}))
  Multiply(A=@n11.A, B=DepthFade(Opacity=Multiply(A=@n15, B=@n18), FadeDistance=Constant() {r=100}))
  ScalarParameter() {defaultValue=1, parameterName=s}
 ]
}
